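# Sound cue for the wrong-number sound of a rotary phone quest in Athena.
# A Random node picks between two wave variations with equal weight, then a Modulator adds slight pitch and volume variation.
# Playback is spatialized through the shared gadget-activate attenuation and routed to the in-game mission sound class.
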SoundCue /Figment_S09/Athena/Sounds/Quests/RotaryPhone/Athena_Quest_Rotary_WrongNumber_Cue
root=Modulator_0
nodes [
 {id=Modulator_0, children [Random_1], properties {pitchMax=1, pitchMin=0.99, volumeMax=1, volumeMin=0.9}}
 {id=Random_1, children [WavePlayer_0,WavePlayer_3], properties {weights [1,1]}}
 {id=WavePlayer_0, properties {soundWaveAssetPtr=&/Athena_Quest_Rotary_WrongNumber_01}}
 {id=WavePlayer_3, properties {soundWaveAssetPtr=&/Athena_Quest_Rotary_WrongNumber_02}}
]
properties {
 attenuationSettings=&Fort_GadgetActivate_Att
 maxDistance=4000
 soundClassObject=/Game/Sounds/FortSoundClassesAndModes/Fort_Mission_IngameSounds
 volumeMultiplier=1
}
